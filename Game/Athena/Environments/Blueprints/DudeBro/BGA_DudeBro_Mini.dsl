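; Mini DudeBro geyser vent actor that launches players upward and affects vehicles, projectiles, C4 and physics balls entering its cone.
; Pawns get vent events, gameplay cues and an optional boost, deferred until they leave any vehicle; thrown objects have gravity scale altered.
; Server overlaps set a timed active state, whose RepNotify animates cone materials, steam and audio on clients.
; A curve-table value can disable the vent at match start.
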(event EventBeginPlay
  (bind _loopingambaudio (Variables|Default|GetLoopingAmbAudio))
  (bind _s_fx_geysercone (Variables|Default|GetS_FX_GeyserCone))
  ;[1 1 1] Begin Play
  (Utilities|FlowControl|Sequence
    (:then_0
      (Utilities|IsValid (Variables|/Script/Engine.StaticMeshComponent|GetStaticMesh (Variables|Default|GetS_FX_GeyserCone))
        (:Is%20Valid
          (Rendering|Material|SetScalarParameterValueonMaterials _s_fx_geysercone "UVScale" (Math|Random|RandomFloatinRange 2.0 4.0))
          (Rendering|Material|SetScalarParameterValueonMaterials _s_fx_geysercone "SlantAmount" (Math|Random|RandomFloatinRange -5e+02 5e+02))
          (Rendering|Material|SetScalarParameterValueonMaterials _s_fx_geysercone "WPO_Amount" (Math|Random|RandomFloatinRange 0.0 5e+02))
          (Rendering|Material|SetScalarParameterValueonMaterials _s_fx_geysercone "WorldTexUVScale" (Math|Random|RandomFloatinRange 0.5 2.0))
          (bind _output_get (Variables|Default|SetGeyserContrast (Math|Random|RandomFloatinRange 1.0 3.0)))
          (Rendering|Material|SetScalarParameterValueonMaterials _s_fx_geysercone "Top Edge Contrast" _output_get))))
    (:then_1
      (Utilities|IsValid (Variables|/Script/CascadeRuntime.ParticleSystemComponent|GetTemplate (Variables|Default|GetP_GeyserSteamBottom))
        (:Is%20Valid
          (Effects|Components|ParticleSystem|SetFloatParameter (Variables|Default|GetP_GeyserSteamBottom) "SpeedSpawn" (Math|Random|RandomFloatinRange 0.5 2.0)))))
    (:then_2
      (Networking|CanExecuteCosmeticEvents
        (:True
          (if (Variables|Default|GetAllowEffects)
            (if (Variables|Default|GetisDisabled)
              (Components|Activation|Deactivate _loopingambaudio)
              (else
                (Components|Activation|Activate _loopingambaudio false)))
            (else
              (Components|DestroyComponent (Variables|Default|GetP_GeyserSteamBottom)))))))
    (:then_3
      (Class|FortAsyncActionAuthorityMatchReady|MatchReadyAsync
        (:OnReady
          (bind _outxy (DataTable|EvaluateCurveTableRow "/Game/Athena/Balance/DataTables/AthenaGameData.AthenaGameData" "Default.Vents.Disabled")
            (:RowFound
              (if (Variables|Default|GetisDisabled)
                _
                (elif (Math|Conversions|ToBoolean(Integer) (Math|Float|Truncate _outxy))
                  (Variables|Default|SetisDisabled true))))))))
    (:then_4
      (Variables|Default|SetSavedConeScale (Variables|/Script/Engine.SceneComponent|GetRelativeScale3D (Variables|Default|GetS_FX_GeyserCone))))))

(event Custom|OnPawnVehicleStateChanged
  (params (FortPlayerPawn :type Fort%20Player%20Pawn%20Object%20Reference) (NewVehicle :type Actor%20Object%20Reference) (OldVehicle :type Actor%20Object%20Reference))
  ; Called when the pawn is inside the overlap area, and its vehicle state changes.
  (if (Utilities|Array|ContainsItem (Variables|Default|GetPlayersInsideZone) FortPlayerPawn)
    (Utilities|IsValid NewVehicle
      (:Is%20Not%20Valid
        (Vehicle|UnbindEventfromOnVehicleStateChangedEvent FortPlayerPawn (EventDispatchers|CreateEvent :!event Custom|OnPawnVehicleStateChanged))
        (CallFunction|LaunchPawnfromVent FortPlayerPawn)))))

(event OnComponentBeginOverlap(S_FX_GeyserCone)
  (params (OverlappedComponent :type Primitive%20Component%20Object%20Reference) (OtherActor :type Actor%20Object%20Reference) (OtherComp :type Primitive%20Component%20Object%20Reference) (OtherBodyIndex :type Integer) (bFromSweep :type Boolean) (SweepResult :type Hit%20Result%20Structure%20%28by%20ref%29))
  ;[1 1 1] Begin Overlap
  (Utilities|FlowControl|Sequence
    (:then_0
      (bind _asfort_player_pawn (Utilities|Casting|CastToFortPlayerPawn OtherActor)
        (:then
          (if (not (GameplayTags|HasAnyMatchingGameplayTags _asfort_player_pawn (Variables|Default|GetTC_IgnoreVent)))
            (Utilities|Array|AddUnique (Variables|Default|GetPlayersInsideZone) _asfort_player_pawn)
            (bind _ignorepawn (Default|ShouldIgnorePawninVehicle _asfort_player_pawn))
            (if (not _ignorepawn)
              (CallFunction|LaunchPawnfromVent _asfort_player_pawn)
              (else
                (Vehicle|AssignOnVehicleStateChangedEvent _asfort_player_pawn (EventDispatchers|CreateEvent :!event Custom|OnPawnVehicleStateChanged))))))
        (:CastFailed
          (bind _asfort_vehicle_interface (Utilities|Casting|CastToFortVehicleInterface OtherActor)
            (:then
              (bind _returnvalue (GameplayTags|HasAnyMatchingGameplayTags(Message) OtherActor (Variables|Default|GetTC_IgnoreVent)))
              (if (not _returnvalue)
                (CallFunction|OnVehicleEnterZone (Vehicle|GetActor _asfort_vehicle_interface) -0.6)
                (Vehicle|SetIgnoreAllFallingDamage _asfort_vehicle_interface true)
                (if (Variables|Default|GetShouldBoost)
                  (Audio|BroadcastAbilityImpactatLocation 0 (Vehicle|GetActor _asfort_vehicle_interface) (+ (Vehicle|GetVehicleVelocity _asfort_vehicle_interface) (* (Transformation|GetActorUpVector) (Variables|Default|GetCharacterLaunchVelocity))) 777.0))))
            (:CastFailed
              (bind _asfort_projectile_base (Utilities|Casting|CastToFortProjectileBase OtherActor)
                (:then
                  (CallFunction|ProjEnter _asfort_projectile_base))
                (:CastFailed
                  (bind _asb_bga_athena_c4 (Utilities|Casting|CastToB_BGA_Athena_C4 OtherActor)
                    (:then
                      (Variables|/Script/Engine.ProjectileMovementComponent|SetProjectileGravityScale (* (Variables|/Script/Engine.ProjectileMovementComponent|GetProjectileGravityScale (Variables|/Game/Athena/Items/Consumables/C4/B_BGA_Athena_C4.B_BGA_Athena_C4_C|GetProjectileMovement1 _asb_bga_athena_c4)) -1.0) (Variables|/Game/Athena/Items/Consumables/C4/B_BGA_Athena_C4.B_BGA_Athena_C4_C|GetProjectileMovement1 _asb_bga_athena_c4)))
                    (:CastFailed
                      (bind _asphysics_ball_master (Utilities|Casting|CastToPhysicsBall_Master OtherActor))
                      (CallFunction|ToyEnter _asphysics_ball_master))))))))))
    (:then_1
      (Networking|SwitchHasAuthority
        (:Authority
          (if (Variables|Default|GetAllowEffects)
            (CallFunction|SetGeiserActive)))))))

(event OnComponentEndOverlap(S_FX_GeyserCone)
  (params (OverlappedComponent :type Primitive%20Component%20Object%20Reference) (OtherActor :type Actor%20Object%20Reference) (OtherComp :type Primitive%20Component%20Object%20Reference) (OtherBodyIndex :type Integer))
  ;[1 1 1] End Overlap
  (bind _asfort_player_pawn (Utilities|Casting|CastToFortPlayerPawn OtherActor)
    (:then
      (Utilities|Array|RemoveItem (Variables|Default|GetPlayersInsideZone) _asfort_player_pawn)
      (Vehicle|UnbindEventfromOnVehicleStateChangedEvent _asfort_player_pawn (EventDispatchers|CreateEvent :!event Custom|OnPawnVehicleStateChanged))
      (Ability|SendGameplayEventtoActor _asfort_player_pawn "(TagName=\"Abilities.Generic.Vent.Exit\")")
      (Variables|/Script/FortniteGame.FortPawn|SetbIgnoreNextFallingDamage true _asfort_player_pawn)
      (CallFunction|RemoveDudeBroGameplayCue _asfort_player_pawn))
    (:CastFailed
      (bind _asfort_vehicle_interface (Utilities|Casting|CastToFortVehicleInterface OtherActor)
        (:then
          (CallFunction|OnVehicleExitZone (Vehicle|GetActor _asfort_vehicle_interface))
          (Vehicle|SetIgnoreAllFallingDamage _asfort_vehicle_interface false)
          (Vehicle|SetIgnoreNextFallingDamage _asfort_vehicle_interface true))
        (:CastFailed
          (bind _asfort_projectile_base (Utilities|Casting|CastToFortProjectileBase OtherActor)
            (:then
              (CallFunction|ProjExit _asfort_projectile_base))
            (:CastFailed
              (bind _asb_bga_athena_c4 (Utilities|Casting|CastToB_BGA_Athena_C4 OtherActor)
                (:then
                  (Variables|/Script/Engine.ProjectileMovementComponent|SetProjectileGravityScale (* (Variables|/Script/Engine.ProjectileMovementComponent|GetProjectileGravityScale (Variables|/Game/Athena/Items/Consumables/C4/B_BGA_Athena_C4.B_BGA_Athena_C4_C|GetProjectileMovement1 _asb_bga_athena_c4)) -1.0) (Variables|/Game/Athena/Items/Consumables/C4/B_BGA_Athena_C4.B_BGA_Athena_C4_C|GetProjectileMovement1 _asb_bga_athena_c4)))
                (:CastFailed
                  (bind _asphysics_ball_master (Utilities|Casting|CastToPhysicsBall_Master OtherActor))
                  (CallFunction|ToyExit _asphysics_ball_master))))))))))

(event Custom|SetGeiserActive
  (Utilities|Time|SetTimerbyEvent (delegate Custom|GeiserTimeout) 5.0)
  (if (not (Variables|Default|GetIsActive))
    (Variables|Default|SetIsActive true)))

(event Custom|GeiserTimeout
  (Variables|Default|SetIsActive false))

(event Custom|ActiveGeyser
  (bind _s_fx_geysercone (Variables|Default|GetS_FX_GeyserCone))
  (bind _s_fx_geysercone_2 (Variables|Default|GetS_FX_GeyserCone))
  ;[1 1 1] Activate and Deactivate Geyser Visuals
  (Audio|PlaySoundatLocation (Variables|Default|GetSoundOnActivate) (Transformation|GetWorldLocation (Variables|Default|GetSM_CorruptedVent_A)))
  (Components|Activation|Activate (Variables|Default|GetP_GeyserSteam) false)
  (Utilities|IsValid (Variables|/Script/Engine.StaticMeshComponent|GetStaticMesh (Variables|Default|GetS_FX_GeyserCone))
    (:Is%20Valid
      (Rendering|Material|SetScalarParameterValueonMaterials _s_fx_geysercone "UVScale" 1.0)
      (Rendering|Material|SetScalarParameterValueonMaterials _s_fx_geysercone "WPO_Contrast" 3.0)
      (Rendering|Material|SetVectorParameterValueonMaterials _s_fx_geysercone "WorldPannerSpeed" "0.000000,2.000000,0.000000")
      (Rendering|Material|SetScalarParameterValueonMaterials _s_fx_geysercone "WPO_Amount" 1.5e+03)
      (Rendering|Material|SetScalarParameterValueonMaterials _s_fx_geysercone "WorldTexUVScale" 0.0)
      (Utilities|IsValid (Variables|/Script/CascadeRuntime.ParticleSystemComponent|GetTemplate (Variables|Default|GetP_GeyserSteamBottom))
        (:Is%20Valid
          (Effects|Components|ParticleSystem|SetFloatParameter (Variables|Default|GetP_GeyserSteamBottom) "SpeedSpawn" 2.0))
        (:Is%20Not%20Valid
          _))
      (bind (_direction _lerp_control) (Timeline|Grow Play)
        (:Update
          (Audio|Parameter|SetFloatParameter (Variables|Default|GetLoopingAmbAudio) "Intensity" _lerp_control :!class /Script/AudioExtensions.AudioParameterControllerInterface)
          (Utilities|IsValid (Variables|/Script/Engine.StaticMeshComponent|GetStaticMesh (Variables|Default|GetS_FX_GeyserCone))
            (:Is%20Valid
              (Transformation|SetRelativeScale3D _s_fx_geysercone_2 (Math|Vector|Lerp(Vector) (Variables|Default|GetSavedConeScale) (+ (Variables|Default|GetSavedConeScale) "0, 0,0.050000") _lerp_control))
              (Rendering|Material|SetScalarParameterValueonMaterials _s_fx_geysercone_2 "Top Edge Contrast" (Math|Float|Lerp (Variables|Default|GetGeyserContrast) 0.5 _lerp_control))
              (Rendering|Material|SetScalarParameterValueonMaterials _s_fx_geysercone_2 "EmissiveContribution" (Math|Float|Lerp 0.1 0.3 _lerp_control)))))))
    (:Is%20Not%20Valid
      _)))

(event Custom|DeactivateGeyser
  (bind _s_fx_geysercone (Variables|Default|GetS_FX_GeyserCone))
  (bind _s_fx_geysercone_2 (Variables|Default|GetS_FX_GeyserCone))
  ;[1 1 1] Activate and Deactivate Geyser Visuals
  (Audio|PlaySoundatLocation (Variables|Default|GetSoundOnDeactivate) (Transformation|GetWorldLocation (Variables|Default|GetSM_CorruptedVent_A)))
  (Components|Activation|Deactivate (Variables|Default|GetP_GeyserSteam))
  (Utilities|IsValid (Variables|/Script/Engine.StaticMeshComponent|GetStaticMesh (Variables|Default|GetS_FX_GeyserCone))
    (:Is%20Valid
      (Rendering|Material|SetScalarParameterValueonMaterials _s_fx_geysercone "UVScale" (Math|Random|RandomFloatinRange 2.0 4.0))
      (Rendering|Material|SetScalarParameterValueonMaterials _s_fx_geysercone "WPO_Contrast" (Math|Random|RandomFloatinRange 1.0 2.0))
      (Rendering|Material|SetScalarParameterValueonMaterials _s_fx_geysercone "WPO_Amount" (Math|Random|RandomFloatinRange 0.0 5e+02))
      (Rendering|Material|SetScalarParameterValueonMaterials _s_fx_geysercone "WorldTexUVScale" (Math|Random|RandomFloatinRange 0.5 2.0))
      (Variables|Default|SetGeyserContrast (Math|Random|RandomFloatinRange 1.0 3.0))
      (Utilities|IsValid (Variables|/Script/CascadeRuntime.ParticleSystemComponent|GetTemplate (Variables|Default|GetP_GeyserSteamBottom))
        (:Is%20Valid
          (Effects|Components|ParticleSystem|SetFloatParameter (Variables|Default|GetP_GeyserSteamBottom) "SpeedSpawn" (Math|Random|RandomFloatinRange 0.5 2.0)))
        (:Is%20Not%20Valid
          _))
      (Timeline|Grow Reverse))
    (:Is%20Not%20Valid
      _)))

(fn AddDudeBroGameplayCue
  (params (InFortPlayerPawn :type Fort%20Player%20Pawn%20Object%20Reference))
  (bind _returnvalue (Ability|GameplayCue|MakeGameplayCueParameters :NormalizedMagnitude 0.0 :RawMagnitude 0.0 :EffectContext "()" :AggregatedSourceTags "(GameplayTags=)" :AggregatedTargetTags "(GameplayTags=)" :Location "0,0,0" :Normal "0,0,0" :GameplayEffectLevel 1 :AbilityLevel 1 :bReplicateLocationWhenUsingMinimalRepProxy false))
  (bind (_returnvalue_x _returnvalue_y _returnvalue_z) (Math|Vector|BreakVector (Transformation|GetActorLocation)))
  (bind _returnvalue_2 (+ _returnvalue_z 8343.651367))
  (bind (_returnvalue_x_2 _returnvalue_y_2 _returnvalue_z_2) (Math|Vector|BreakVector (Transformation|GetActorLocation (Variables|GetInFortPlayerPawn))))
  (bind _returnvalue_3 (<= _returnvalue_z_2 _returnvalue_2))
  (bind _returnvalue_4 (Math|Conversions|ToFloat(Boolean) _returnvalue_3))
  (bind _returnvalue_5 (Ability|GameplayCue|MakeGameplayCueParameters :NormalizedMagnitude _returnvalue_4 :RawMagnitude 0.0 :EffectContext "()" :AggregatedSourceTags "(GameplayTags=)" :AggregatedTargetTags "(GameplayTags=)" :Location "0,0,0" :Normal "0,0,0" :GameplayEffectLevel 1 :AbilityLevel 1 :bReplicateLocationWhenUsingMinimalRepProxy false))
  (bind _infortplayerpawn (Variables|GetInFortPlayerPawn))
  (bind _dolocalgameplaycues (Variables|Default|GetDoLocalGameplayCues))
  (if (and (== (GameplayEffects|GetGameplayEffectCount (Ability|GetAbilitySystemComponent InFortPlayerPawn) "/Game/Athena/Environments/Blueprints/DudeBro/GE_DudeBro_Vent_Area_Effect.GE_DudeBro_Vent_Area_Effect_C" 0 true) 0) (Variables|Default|GetShouldBoost))
    (if _dolocalgameplaycues
      (Projectile|AddGameplayCue%28Looping%29Local (Variables|GetInFortPlayerPawn) (Variables|Default|GetVentCue) _returnvalue_5)
      (else
        (Networking|SwitchHasAuthority
          (:Authority
            (GameplayCue|AddGameplayCueOnActor(Looping) (Variables|GetInFortPlayerPawn) (Variables|Default|GetVentCue) _returnvalue_5)))))
    (elif _dolocalgameplaycues
      (Projectile|AddGameplayCue%28Looping%29Local (Variables|GetInFortPlayerPawn) (Variables|Default|GetVolcanoCue) _returnvalue)
      (else
        (Networking|SwitchHasAuthority
          (:Authority
            (GameplayCue|AddGameplayCueOnActor(Looping) (Variables|GetInFortPlayerPawn) (Variables|Default|GetVolcanoCue) _returnvalue)))))))

(fn LaunchPawnFromVent
  (params (Pawn :type Fort%20Player%20Pawn%20Object%20Reference))
  (bind _returnvalue (Fortnite|GetFortPlayerControllerfromActor Pawn))
  (bind _shouldboost (Variables|Default|GetShouldBoost))
  (Utilities|FlowControl|Sequence
    (:then_0
      (Ability|SendGameplayEventtoActor Pawn "(TagName=\"Abilities.Generic.Vent.Enter\")" (Utilities|Struct|MakeGameplayEventData :Target self :ContextHandle "()" :InstigatorTags "(GameplayTags=)" :TargetTags "(GameplayTags=)" :EventMagnitude (Variables|Default|GetUpwardLaunchVelocity) :TargetData "()"))
      (Development|PrintString (Utilities|String|ToString(Boolean) _shouldboost) false false)
      (if _shouldboost
        (Character|LaunchCharacter Pawn (* (Transformation|GetActorUpVector) (Variables|Default|GetCharacterLaunchVelocity)) (Variables|Default|GetXYOverride) (Variables|Default|GetZOverride))
        ;[1 1 1] Use Geysers in Different Matches Quest hook in
        (if (Networking|IsServer)
          (Utilities|IsValid _returnvalue
            (:Is%20Valid
              (Quest|SendComplexCustomStatEvent (Quests|GetQuestManager _returnvalue "Invalid") 0 "(GameplayTags=((TagName=\"Athena.Quests.Guyser\")))" "(GameplayTags=)" 1))))))
    (:then_1
      (CallFunction|AddDudeBroGameplayCue Pawn))))

(fn OnRep_IsActive
  (Networking|SwitchHasAuthority
    (:Remote
      (if (Variables|Default|GetIsActive)
        (CallFunction|ActiveGeyser)
        (else
          (CallFunction|DeactivateGeyser))))))

(fn OnRep_isDisabled
  (if (Variables|Default|GetisDisabled)
    (CallFunction|DeactivateGeyser)
    (Collision|SetCollisionEnabled (Variables|Default|GetS_FX_GeyserCone) "NoCollision")
    (Components|Activation|Deactivate (Variables|Default|GetP_GeyserSteamBottom))
    (Rendering|SetActorHiddenInGame self true)
    (else
      (CallFunction|ActiveGeyser)
      (Collision|SetCollisionEnabled (Variables|Default|GetS_FX_GeyserCone) "QueryOnly")
      (Rendering|SetActorHiddenInGame self false)
      (Networking|CanExecuteCosmeticEvents
        (:True
          (Components|Activation|Activate (Variables|Default|GetP_GeyserSteamBottom) true))))))

(fn ProjEnter
  (params (Projectile :type Fort%20Projectile%20Base%20Object%20Reference))
  (bind _projectilemovementcomponent (Variables|/Script/FortniteGame.FortProjectileBase|GetProjectileMovementComponent Projectile))
  (bind _projectilegravityscale (Variables|/Script/Engine.ProjectileMovementComponent|GetProjectileGravityScale _projectilemovementcomponent))
  (if (> _projectilegravityscale 0.0)
    (Variables|/Script/Engine.ProjectileMovementComponent|SetProjectileGravityScale (* _projectilegravityscale -1.0) _projectilemovementcomponent)
    (else
      (Variables|/Script/Engine.ProjectileMovementComponent|SetProjectileGravityScale (+ _projectilegravityscale (Variables|Default|GetZeroGravPrj)) _projectilemovementcomponent))))

(fn ProjExit
  (params (Projectile :type Fort%20Projectile%20Base%20Object%20Reference))
  (bind _projectilemovementcomponent (Variables|/Script/FortniteGame.FortProjectileBase|GetProjectileMovementComponent Projectile))
  (bind _projectilegravityscale (Variables|/Script/Engine.ProjectileMovementComponent|GetProjectileGravityScale _projectilemovementcomponent))
  (if (== _projectilegravityscale (Variables|Default|GetZeroGravPrj))
    (Variables|/Script/Engine.ProjectileMovementComponent|SetProjectileGravityScale (- _projectilegravityscale (Variables|Default|GetZeroGravPrj)) _projectilemovementcomponent)
    (else
      (Variables|/Script/Engine.ProjectileMovementComponent|SetProjectileGravityScale (* _projectilegravityscale -1.0) _projectilemovementcomponent))))

(fn RemoveDudeBroGameplayCue
  (params (InFortPlayerPawn :type Fort%20Player%20Pawn%20Object%20Reference))
  (bind _returnvalue (Ability|GameplayCue|MakeGameplayCueParameters :NormalizedMagnitude 0.0 :RawMagnitude 0.0 :EffectContext "()" :AggregatedSourceTags "(GameplayTags=)" :AggregatedTargetTags "(GameplayTags=)" :Location "0,0,0" :Normal "0,0,0" :GameplayEffectLevel 1 :AbilityLevel 1 :bReplicateLocationWhenUsingMinimalRepProxy false))
  (bind _returnvalue_2 (Ability|GameplayCue|MakeGameplayCueParameters :NormalizedMagnitude 0.0 :RawMagnitude 0.0 :EffectContext "()" :AggregatedSourceTags "(GameplayTags=)" :AggregatedTargetTags "(GameplayTags=)" :Location "0,0,0" :Normal "0,0,0" :GameplayEffectLevel 1 :AbilityLevel 1 :bReplicateLocationWhenUsingMinimalRepProxy false))
  (if (Variables|Default|GetShouldBoost)
    (if (Variables|Default|GetDoLocalGameplayCues)
      (Projectile|RemoveGameplayCue%28Looping%29Local InFortPlayerPawn (Variables|Default|GetVentCue) _returnvalue_2)
      (else
        (Networking|SwitchHasAuthority
          (:Authority
            (GameplayCue|RemoveGameplayCueOnActor(Looping) InFortPlayerPawn (Variables|Default|GetVentCue) _returnvalue_2)))))
    (else
      (if (Variables|Default|GetDoLocalGameplayCues)
        (Projectile|RemoveGameplayCue%28Looping%29Local InFortPlayerPawn (Variables|Default|GetVolcanoCue) _returnvalue)
        (else
          (Networking|SwitchHasAuthority
            (:Authority
              (GameplayCue|RemoveGameplayCueOnActor(Looping) InFortPlayerPawn (Variables|Default|GetVolcanoCue) _returnvalue))))))))

(fn ShouldIgnorePawnInVehicle
  (params (FortPlayerPawn :type Fort%20Player%20Pawn%20Object%20Reference))
  (return (Vehicle|IsinVehicle FortPlayerPawn)))

(fn ToyEnter
  (params (Projectile :type Physics%20Ball%20Master%20Object%20Reference))
  (bind _projectilemovement (Variables|/Game/Building/ActorBlueprints/Prop/PhysicsBall_Master.PhysicsBall_Master_C|GetProjectileMovement Projectile))
  (bind _projectilegravityscale (Variables|/Script/Engine.ProjectileMovementComponent|GetProjectileGravityScale _projectilemovement))
  (if (> _projectilegravityscale 0.0)
    (Variables|/Script/Engine.ProjectileMovementComponent|SetProjectileGravityScale (* _projectilegravityscale -0.5) _projectilemovement)
    (else
      (Variables|/Script/Engine.ProjectileMovementComponent|SetProjectileGravityScale (+ _projectilegravityscale (Variables|Default|GetZeroGravPrj)) _projectilemovement))))

(fn ToyExit
  (params (Projectile :type Physics%20Ball%20Master%20Object%20Reference))
  (bind _projectilemovement (Variables|/Game/Building/ActorBlueprints/Prop/PhysicsBall_Master.PhysicsBall_Master_C|GetProjectileMovement Projectile))
  (bind _projectilegravityscale (Variables|/Script/Engine.ProjectileMovementComponent|GetProjectileGravityScale _projectilemovement))
  (if (== _projectilegravityscale (Variables|Default|GetZeroGravPrj))
    (Variables|/Script/Engine.ProjectileMovementComponent|SetProjectileGravityScale (- _projectilegravityscale (Variables|Default|GetZeroGravPrj)) _projectilemovement)
    (else
      (Variables|/Script/Engine.ProjectileMovementComponent|SetProjectileGravityScale (* _projectilegravityscale -2.0) _projectilemovement))))

(fn ConstructionScript
  (bind _p_geysersteambottom (Variables|Default|GetP_GeyserSteamBottom))
  (|CallParentFunction)
  (if (Variables|Default|GetEditParticleSystems)
    (Utilities|FlowControl|Sequence
      (:then_0
        (Utilities|IsValid (Variables|/Script/Engine.StaticMeshComponent|GetStaticMesh (Variables|Default|GetS_FX_GeyserCone))
          (:Is%20Valid
            (Rendering|Material|SetVectorParameterValueonMaterials (Variables|Default|GetS_FX_GeyserCone) "SulfurBaseColor" (Math|Conversions|ToVector(LinearColor) (Variables|LocalVFXColor|GetGeyser%20Illuminated%20Mesh%20Color))))))
      (:then_1
        (Utilities|IsValid (Variables|/Script/CascadeRuntime.ParticleSystemComponent|GetTemplate (Variables|Default|GetP_GeyserSteamBottom))
          (:Is%20Valid
            (bind _returnvalue (Rendering|Material|CreateNamedDynamicMaterialInstance _p_geysersteambottom "BaseMesh" "/Game/Effects/Fort_Effects/Materials/Water/MI_WaterFall_Basev2_GeyserBase.MI_WaterFall_Basev2_GeyserBase"))
            (Rendering|Material|SetMaterialbyName _p_geysersteambottom "BaseMesh" _returnvalue)
            (Rendering|Material|SetVectorParameterValue _returnvalue "FrenselFakeColor" (Variables|LocalVFXColor|GetBase%20Bubbling%20Mesh%20Color) :!class /Script/Engine.MaterialInstanceDynamic)))))))
