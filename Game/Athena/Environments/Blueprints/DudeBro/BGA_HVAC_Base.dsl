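; Base Blueprint for HVAC air vents that launch players and other objects entering a trigger volume along the vent's up vector.
; Overlaps branch by actor type, handling player dive events, fall-damage immunity and quest hooks, vehicle boosts, physics impulses, and gravity-scale flips for projectiles, C4 and toy balls, reversed on exit.
; Each overlap also plays wind material, particle and sound effects.
; At match ready, a game-data curve can disable the vent.
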(event EventBeginPlay
  ;[1 1 1] Begin Play
  (Utilities|FlowControl|Sequence
    (:then_1
      (if (not (Variables|Default|GetAllowEffects))
        (Components|DestroyComponent (Variables|Default|GetLoopingAmbAudio))))
    (:then_2
      (Class|FortAsyncActionAuthorityMatchReady|MatchReadyAsync
        (:OnReady
          (bind _outxy (DataTable|EvaluateCurveTableRow "/Game/Athena/Balance/DataTables/AthenaGameData.AthenaGameData" "Default.Vents.Disabled")
            (:RowFound
              (if (Variables|Default|GetisDisabled)
                _
                (elif (Math|Conversions|ToBoolean(Integer) (Math|Float|Truncate _outxy))
                  (Variables|Default|SetisDisabled true))))))))))

(event OnComponentBeginOverlap(TriggerVolume)
  (params (OverlappedComponent :type Primitive%20Component%20Object%20Reference) (OtherActor :type Actor%20Object%20Reference) (OtherComp :type Primitive%20Component%20Object%20Reference) (OtherBodyIndex :type Integer) (bFromSweep :type Boolean) (SweepResult :type Hit%20Result%20Structure%20%28by%20ref%29))
  ;[1 1 1] Begin Overlap
  (Utilities|FlowControl|Sequence
    (:then_0
      (bind _asfort_player_pawn (Utilities|Casting|CastToFortPlayerPawn OtherActor)
        (:then
          (if (Variables|Default|GetPutInDiveMode)
            (Ability|SendGameplayEventtoActor _asfort_player_pawn "(TagName=\"Abilities.Generic.Vent.Enter\")" (Utilities|Struct|MakeGameplayEventData :Instigator _asfort_player_pawn :Target self :ContextHandle "()" :InstigatorTags "(GameplayTags=)" :TargetTags "(GameplayTags=)" :EventMagnitude (Variables|Default|GetUpwardLaunchVelocity) :TargetData "()")))
          (if (Variables|Default|GetShouldBoost)
            (Character|LaunchCharacter _asfort_player_pawn (* (Transformation|GetActorUpVector) (Variables|Default|GetCharacterLaunchVelocity)) (Variables|Default|GetXYOverride) (Variables|Default|GetZOverride))
            (Environment|ApplyGlobalEnvironmentGameplayEffecttoActor _asfort_player_pawn "/Game/Athena/Environments/Blueprints/DudeBro/GE_Athena_FallDmgImmune_FX.GE_Athena_FallDmgImmune_FX_C" 0 "(GameplayTags=)")
            ;[1 1 1] Use Geysers in Different Matches Quest hook in
            (if (Networking|IsServer)
              (Quest|SendComplexCustomStatEvent (Quests|GetQuestManager (Fortnite|GetFortPlayerControllerfromActor _asfort_player_pawn) "Invalid") 0 "(GameplayTags=((TagName=\"Athena.Quests.HVAC\")))" "(GameplayTags=)" 1)
              ; Calls a custom verb when looper is launched by Air Vent
              (CustomVerb|BroadcastCustomVerbMessage (CustomVerb|MakeCustomVerbMessage (Fortnite|GetFortPlayerControllerfromActor _asfort_player_pawn) "(TagName=\"CustomVerb.Figment.AirVentLaunch\")")))))
        (:CastFailed
          (bind _asfort_vehicle_interface (Utilities|Casting|CastToFortVehicleInterface OtherActor)
            (:then
              (Vehicle|SetIgnoreAllFallingDamage _asfort_vehicle_interface true)
              (if (Variables|Default|GetShouldBoost)
                (Audio|BroadcastAbilityImpactatLocation 0 (Vehicle|GetActor _asfort_vehicle_interface) (+ (Vehicle|GetVehicleVelocity _asfort_vehicle_interface) (* (Transformation|GetActorUpVector) (Variables|Default|GetCharacterLaunchVelocity))) 777.0)))
            (:CastFailed
              (bind _asfort_projectile_base (Utilities|Casting|CastToFortProjectileBase OtherActor)
                (:then
                  (CallFunction|ProjEnter _asfort_projectile_base))
                (:CastFailed
                  (bind _asb_bga_athena_c4 (Utilities|Casting|CastToB_BGA_Athena_C4 OtherActor)
                    (:then
                      (Variables|/Script/Engine.ProjectileMovementComponent|SetProjectileGravityScale (* (Variables|/Script/Engine.ProjectileMovementComponent|GetProjectileGravityScale (Variables|/Game/Athena/Items/Consumables/C4/B_BGA_Athena_C4.B_BGA_Athena_C4_C|GetProjectileMovement1 _asb_bga_athena_c4)) -1.0) (Variables|/Game/Athena/Items/Consumables/C4/B_BGA_Athena_C4.B_BGA_Athena_C4_C|GetProjectileMovement1 _asb_bga_athena_c4)))
                    (:CastFailed
                      (bind _asphysics_ball_master (Utilities|Casting|CastToPhysicsBall_Master OtherActor)
                        (:then
                          (CallFunction|ToyEnter _asphysics_ball_master))
                        (:CastFailed
                          (bind _asfort_physics_pawn_object (Utilities|Casting|CastToFortPhysicsPawnObject OtherActor))
                          (PhysicsObject|ApplyImpulse _asfort_physics_pawn_object self (* (Transformation|GetActorUpVector) (Variables|Default|GetCharacterLaunchVelocity))))))))))))))
    (:then_1
      (CallFunction|CEPlayWind OtherActor))))

(event OnComponentEndOverlap(TriggerVolume)
  (params (OverlappedComponent :type Primitive%20Component%20Object%20Reference) (OtherActor :type Actor%20Object%20Reference) (OtherComp :type Primitive%20Component%20Object%20Reference) (OtherBodyIndex :type Integer))
  ;[1 1 1] End Overlap
  (bind _asfort_player_pawn (Utilities|Casting|CastToFortPlayerPawn OtherActor)
    (:then
      (Ability|SendGameplayEventtoActor _asfort_player_pawn "(TagName=\"Abilities.Generic.Vent.Exit\")")
      (Variables|/Script/FortniteGame.FortPawn|SetbIgnoreNextFallingDamage true _asfort_player_pawn)
      (if (Variables|Default|GetShouldBoost)
        (Projectile|RemoveGameplayCue%28Looping%29Local _asfort_player_pawn (Variables|Default|GetVentCue) (Ability|GameplayCue|MakeGameplayCueParameters :NormalizedMagnitude 0.0 :RawMagnitude 0.0 :EffectContext "()" :AggregatedSourceTags "(GameplayTags=)" :AggregatedTargetTags "(GameplayTags=)" :Location "0,0,0" :Normal "0,0,0" :GameplayEffectLevel 1 :AbilityLevel 1 :bReplicateLocationWhenUsingMinimalRepProxy false))
        (else
          (Projectile|RemoveGameplayCue%28Looping%29Local _asfort_player_pawn (Variables|Default|GetVolcanoCue) (Ability|GameplayCue|MakeGameplayCueParameters :NormalizedMagnitude 0.0 :RawMagnitude 0.0 :EffectContext "()" :AggregatedSourceTags "(GameplayTags=)" :AggregatedTargetTags "(GameplayTags=)" :Location "0,0,0" :Normal "0,0,0" :GameplayEffectLevel 1 :AbilityLevel 1 :bReplicateLocationWhenUsingMinimalRepProxy false)))))
    (:CastFailed
      (bind _asfort_vehicle_interface (Utilities|Casting|CastToFortVehicleInterface OtherActor)
        (:then
          (Vehicle|SetIgnoreAllFallingDamage _asfort_vehicle_interface false)
          (Vehicle|SetIgnoreNextFallingDamage _asfort_vehicle_interface true))
        (:CastFailed
          (bind _asfort_projectile_base (Utilities|Casting|CastToFortProjectileBase OtherActor)
            (:then
              (CallFunction|ProjExit _asfort_projectile_base))
            (:CastFailed
              (bind _asb_bga_athena_c4 (Utilities|Casting|CastToB_BGA_Athena_C4 OtherActor)
                (:then
                  (Variables|/Script/Engine.ProjectileMovementComponent|SetProjectileGravityScale (* (Variables|/Script/Engine.ProjectileMovementComponent|GetProjectileGravityScale (Variables|/Game/Athena/Items/Consumables/C4/B_BGA_Athena_C4.B_BGA_Athena_C4_C|GetProjectileMovement1 _asb_bga_athena_c4)) -1.0) (Variables|/Game/Athena/Items/Consumables/C4/B_BGA_Athena_C4.B_BGA_Athena_C4_C|GetProjectileMovement1 _asb_bga_athena_c4)))
                (:CastFailed
                  (bind _asphysics_ball_master (Utilities|Casting|CastToPhysicsBall_Master OtherActor))
                  (CallFunction|ToyExit _asphysics_ball_master))))))))))

(event Custom|CE_PlayWind
  (params (Actor :type Actor%20Object%20Reference))
  (bind _sm_hvac_fx (Variables|Default|GetSM_HVAC_FX))
  (bind _returnvalue (Transformation|GetWorldRotation _sm_hvac_fx))
  (bind _fx_mid (Variables|Default|GetFX_MID))
  ;[0 0 0] FX
  (Utilities|FlowControl|Sequence
    (:then_0
      (Utilities|FlowControl|DoOnce false
        (:Completed
          (bind _returnvalue_2 (Rendering|Material|CreateDynamicMaterialInstance (Variables|Default|GetSM_HVAC_FX) 0 "/Game/Athena/Environments/Blueprints/DudeBro/HVAC_FX/FX_Wind.FX_Wind" "None" :!class /Script/Engine.PrimitiveComponent))
          (Variables|Default|SetFX_MID _returnvalue_2))))
    (:then_1
      (bind (_direction _bp_speeduptimer _bp_wpo_movingup _bp_intensity) (Timeline|Timeline_0 PlayFromStart)
        (:Update
          (Rendering|Material|SetScalarParameterValue _fx_mid "BP_SpeedUpTimer" _bp_speeduptimer :!class /Script/Engine.MaterialInstanceDynamic)
          (Rendering|Material|SetScalarParameterValue _fx_mid "BP_WPO_MovingUP" _bp_wpo_movingup :!class /Script/Engine.MaterialInstanceDynamic)
          (Rendering|Material|SetScalarParameterValue _fx_mid "BP_Intensity" _bp_intensity :!class /Script/Engine.MaterialInstanceDynamic))))
    (:then_2
      (bind _asfort_player_pawn (Utilities|Casting|CastToFortPlayerPawn Actor))
      (GameplayCue|ExecuteGameplayCue%28Burst%29Local _asfort_player_pawn (Variables|Default|GetWindCue) (Ability|GameplayCue|MakeGameplayCueParameters :Normal (Math|Rotator|GetRotationXVector _returnvalue)) true)
      (Effects|SpawnEmitteratLocation "/Game/Athena/Environments/Blueprints/DudeBro/HVAC_FX/P_WindBasement.P_WindBasement" (Transformation|GetWorldLocation (Variables|Transformation|GetRootComponent _asfort_player_pawn)) _returnvalue)
      (Audio|PlaySoundatLocation (Variables|Default|GetSoundOnActivate) (Transformation|GetWorldLocation (Variables|Default|GetLoopingAmbAudio)))
      (if (Pawn|IsLocallyControlled _asfort_player_pawn)
        (Audio|PlaySound2D "/Game/Athena/Sounds/Ambient/HVAC/HVAC_Activate_1P_Cue.HVAC_Activate_1P_Cue")))))

(fn OnRep_isDisabled
  (if (Variables|Default|GetisDisabled)
    (Components|Activation|Deactivate (Variables|Default|GetLoopingAmbAudio))
    (Rendering|SetActorHiddenInGame self true)))

(fn ProjEnter
  (params (Projectile :type Fort%20Projectile%20Base%20Object%20Reference))
  (bind _projectilemovementcomponent (Variables|/Script/FortniteGame.FortProjectileBase|GetProjectileMovementComponent Projectile))
  (bind _projectilegravityscale (Variables|/Script/Engine.ProjectileMovementComponent|GetProjectileGravityScale _projectilemovementcomponent))
  (if (> _projectilegravityscale 0.0)
    (Variables|/Script/Engine.ProjectileMovementComponent|SetProjectileGravityScale (* _projectilegravityscale -1.0) _projectilemovementcomponent)
    (else
      (Variables|/Script/Engine.ProjectileMovementComponent|SetProjectileGravityScale (+ _projectilegravityscale (Variables|Default|GetZeroGravPrj)) _projectilemovementcomponent))))

(fn ProjExit
  (params (Projectile :type Fort%20Projectile%20Base%20Object%20Reference))
  (bind _projectilemovementcomponent (Variables|/Script/FortniteGame.FortProjectileBase|GetProjectileMovementComponent Projectile))
  (bind _projectilegravityscale (Variables|/Script/Engine.ProjectileMovementComponent|GetProjectileGravityScale _projectilemovementcomponent))
  (if (== _projectilegravityscale (Variables|Default|GetZeroGravPrj))
    (Variables|/Script/Engine.ProjectileMovementComponent|SetProjectileGravityScale (- _projectilegravityscale (Variables|Default|GetZeroGravPrj)) _projectilemovementcomponent)
    (else
      (Variables|/Script/Engine.ProjectileMovementComponent|SetProjectileGravityScale (* _projectilegravityscale -1.0) _projectilemovementcomponent))))

(fn ToyEnter
  (params (Projectile :type Physics%20Ball%20Master%20Object%20Reference))
  (bind _projectilemovement (Variables|/Game/Building/ActorBlueprints/Prop/PhysicsBall_Master.PhysicsBall_Master_C|GetProjectileMovement Projectile))
  (bind _projectilegravityscale (Variables|/Script/Engine.ProjectileMovementComponent|GetProjectileGravityScale _projectilemovement))
  (if (> _projectilegravityscale 0.0)
    (Variables|/Script/Engine.ProjectileMovementComponent|SetProjectileGravityScale (* _projectilegravityscale -0.5) _projectilemovement)
    (else
      (Variables|/Script/Engine.ProjectileMovementComponent|SetProjectileGravityScale (+ _projectilegravityscale (Variables|Default|GetZeroGravPrj)) _projectilemovement))))

(fn ToyExit
  (params (Projectile :type Physics%20Ball%20Master%20Object%20Reference))
  (bind _projectilemovement (Variables|/Game/Building/ActorBlueprints/Prop/PhysicsBall_Master.PhysicsBall_Master_C|GetProjectileMovement Projectile))
  (bind _projectilegravityscale (Variables|/Script/Engine.ProjectileMovementComponent|GetProjectileGravityScale _projectilemovement))
  (if (== _projectilegravityscale (Variables|Default|GetZeroGravPrj))
    (Variables|/Script/Engine.ProjectileMovementComponent|SetProjectileGravityScale (- _projectilegravityscale (Variables|Default|GetZeroGravPrj)) _projectilemovement)
    (else
      (Variables|/Script/Engine.ProjectileMovementComponent|SetProjectileGravityScale (* _projectilegravityscale -2.0) _projectilemovement))))

(fn ConstructionScript
  (|CallParentFunction))
