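; HVAC air vent prop that launches actors entering its trigger along its up vector.
; Overlaps branch by cast: pawns, vehicles, projectiles, C4, physics balls and physics pawn objects each get their own push, with fall-damage protection for pawns and vehicles.
; Boosting players also box-traces to damage objects above and reports quest events.
; Server-triggered wind effects animate a timeline-driven material; a scalable float can disable the prop, which OnRep hides and makes non-colliding.
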(event EventOnDeathPlayEffects
  (params (Damage :type Float%20%28single-precision%29) (DamageTags :type Gameplay%20Tag%20Container%20Structure%20%28by%20ref%29) (Momentum :type Vector :default "0, 0, 0") (HitInfo :type Hit%20Result%20Structure%20%28by%20ref%29) (InstigatedBy :type Fort%20Pawn%20Object%20Reference) (DamageCauser :type Actor%20Object%20Reference) (EffectContext :type Gameplay%20Effect%20Context%20Handle%20Structure))
  (|CallParentFunction Damage DamageTags Momentum HitInfo InstigatedBy DamageCauser EffectContext)
  (Audio|Components|Audio|FadeOut (Variables|Default|GetLoopingAmbAudio) 0.0 0.0 "Linear" :!class /Script/Engine.AudioComponent))

(event EventBeginPlay
  ;[1 1 1] Begin Play
  (Utilities|FlowControl|Sequence
    (:then_1
      (if (not (Variables|Default|GetAllowEffects))
        (Components|DestroyComponent (Variables|Default|GetLoopingAmbAudio))))
    (:then_2
      (Class|FortAsyncActionAuthorityMatchReady|MatchReadyAsync)
      (if (Variables|Default|GetisDisabled)
        _
        (elif (Math|Conversions|ToBoolean(Integer) (Math|Float|Truncate (ScalableFloat|GetValueatLevel (Variables|Default|GetDisabled))))
          (Variables|Default|SetisDisabled true))))))

(event Custom|ResetDonceTrigger
  (params (FlownActor :type Actor%20Object%20Reference))
  ;[1 1 1] Begin Overlap
  (Utilities|Array|RemoveItem (Variables|Default|GetCurrentFlyingActors) FlownActor)
  (Utilities|FlowControl|DoOnce false
    (:Completed
      (Utilities|FlowControl|Sequence
        (:then_0
          (CallFunction|ActorEnteredTrigger (stale OnComponentBeginOverlap(TriggerVolume) _array_element) (stale OnComponentBeginOverlap(TriggerVolume) SweepResult)))
        (:then_1
          (Utilities|FlowControl|Delay 0.2)
          (CallFunction|ResetDonceTrigger (stale OnComponentBeginOverlap(TriggerVolume) _array_element))
          (Utilities|FlowControl|Delay 0.5)
          (CallFunction|ResetDoOnceLaunch))))))

(event OnComponentBeginOverlap(TriggerVolume)
  (params (OverlappedComponent :type Primitive%20Component%20Object%20Reference) (OtherActor :type Actor%20Object%20Reference) (OtherComp :type Primitive%20Component%20Object%20Reference) (OtherBodyIndex :type Integer) (bFromSweep :type Boolean) (SweepResult :type Hit%20Result%20Structure%20%28by%20ref%29))
  ;[1 1 1] Begin Overlap
  (Utilities|Array|Add (Variables|Default|GetCurrentFlyingActors) OtherActor)
  (for _array_element (Variables|Default|GetCurrentFlyingActors)
    (if (== _array_element OtherActor)
      (Utilities|FlowControl|DoOnce false
        (:Completed
          (Utilities|FlowControl|Sequence
            (:then_0
              (CallFunction|ActorEnteredTrigger _array_element SweepResult))
            (:then_1
              (Utilities|FlowControl|Delay 0.2)
              (CallFunction|ResetDonceTrigger _array_element)
              (Utilities|FlowControl|Delay 0.5)
              (CallFunction|ResetDoOnceLaunch))))))))

(event Custom|Actor%20Entered%20Trigger
  (params (Actor :type Actor%20Object%20Reference) (Hit :type Hit%20Result%20Structure))
  ;[1 1 1] Begin Overlap
  (Utilities|FlowControl|Sequence
    (:then_0
      (bind _asfort_player_pawn (Utilities|Casting|CastToFortPlayerPawn Actor)
        (:then
          (if (Variables|Default|GetPutInDiveMode)
            (Ability|SendGameplayEventtoActor _asfort_player_pawn "(TagName=\"Abilities.Generic.Vent.Enter\")" (Utilities|Struct|MakeGameplayEventData :Instigator _asfort_player_pawn :Target self :ContextHandle "()" :InstigatorTags "(GameplayTags=)" :TargetTags "(GameplayTags=)" :EventMagnitude (Variables|Default|GetUpwardLaunchVelocity) :TargetData "()")))
          (if (Variables|Default|GetShouldBoost)
            (Character|LaunchCharacter _asfort_player_pawn (* (Transformation|GetActorUpVector) (* (* (ScalableFloat|GetValueatLevel (Variables|Default|GetLaunchVelocity)) (Variables|UserKnobs%28Creative%29|GetDynamicForceScale)) (Variables|Default|GetLaunch%20Strength%20Modifier))) (Variables|Default|GetXYOverride) (Variables|Default|GetZOverride))
            (Environment|ApplyGlobalEnvironmentGameplayEffecttoActor _asfort_player_pawn "/Game/Athena/Environments/Blueprints/DudeBro/GE_Athena_FallDmgImmune_FX.GE_Athena_FallDmgImmune_FX_C" 0 "(GameplayTags=)")
            ;[1 1 1] Destroy Build Piece on launch if there's one right above HVAC.
            (bind _outhits (Collision|MultiBoxTraceForObjects (+ (Transformation|GetActorLocation) (* (Transformation|GetActorUpVector) (Variables|Default|GetDamageOffset))) (+ (+ (Transformation|GetActorLocation) (* (Transformation|GetActorUpVector) (Variables|Default|GetDamageOffset))) (* (Transformation|GetActorUpVector) (Variables|Default|GetDamageHeight))) "0.000000,65.000000,65.000000" (Math|Conversions|RotationFromXVector (Transformation|GetActorUpVector)) (Variables|Default|GetDestroyObjectTypes)))
            (for _array_element _outhits
              (if (Variables|Actor|GetbCanBeDamaged (.HitActor (Collision|BreakHitResult _array_element)))
                (bind _bdidactorhavetags (Class|AthenaQuestFunctionLibrary|AthenaCheckActorHasTags (.HitActor (Collision|BreakHitResult _array_element)) "(GameplayTags=((TagName=\"Deimos.Spawner\")))"))
                (if (not _bdidactorhavetags)
                  (Environment|ApplyGlobalEnvironmentGameplayEffecttoActor (.HitActor (Collision|BreakHitResult _array_element)) "/Game/Athena/Environments/Blueprints/DudeBro/GE_HVAC_StructureDamage.GE_HVAC_StructureDamage_C"))))
            ;[1 1 1] Use Geysers in Different Matches Quest hook in
            (if (Networking|IsServer)
              (Quest|SendComplexCustomStatEvent (Quests|GetQuestManager (Fortnite|GetFortPlayerControllerfromActor _asfort_player_pawn) "Invalid") 0 "(GameplayTags=((TagName=\"Athena.Quests.HVAC\")))" "(GameplayTags=)" 1)
              ; Calls a custom verb when looper is launched by Air Vent
              (CustomVerb|BroadcastCustomVerbMessage (CustomVerb|MakeCustomVerbMessage (Fortnite|GetFortPlayerControllerfromActor _asfort_player_pawn) "(TagName=\"CustomVerb.Figment.AirVentLaunch\")")))))
        (:CastFailed
          (bind _asfort_vehicle_interface (Utilities|Casting|CastToFortVehicleInterface Actor)
            (:then
              (bind _returnvalue (Vehicle|GetVehicleActor _asfort_vehicle_interface :!class /Script/FortniteGame.FortVehicleInterface))
              (Vehicle|SetIgnoreAllFallingDamage _asfort_vehicle_interface true)
              (if (Variables|Default|GetShouldBoost)
                (Utilities|FlowControl|DoOnce false
                  (:Completed
                    (Audio|BroadcastAbilityImpactatLocation 0 _returnvalue (+ (Vehicle|GetVehicleVelocity _asfort_vehicle_interface) (* (Transformation|GetActorUpVector) (* (ScalableFloat|GetValueatLevel (Variables|Default|GetLaunchVelocity)) (Variables|UserKnobs%28Creative%29|GetDynamicForceScale)))) 777.0)))))
            (:CastFailed
              (bind _asfort_projectile_base (Utilities|Casting|CastToFortProjectileBase Actor)
                (:then
                  (CallFunction|ProjEnter _asfort_projectile_base))
                (:CastFailed
                  (bind _asb_bga_athena_c4 (Utilities|Casting|CastToB_BGA_Athena_C4 Actor)
                    (:then
                      (Variables|/Script/Engine.ProjectileMovementComponent|SetProjectileGravityScale (Math|Float|float*float (Variables|/Script/Engine.ProjectileMovementComponent|GetProjectileGravityScale (Variables|/Game/Athena/Items/Consumables/C4/B_BGA_Athena_C4.B_BGA_Athena_C4_C|GetProjectileMovement1 _asb_bga_athena_c4)) -1.0 (Variables|UserKnobs%28Creative%29|GetDynamicForceScale)) (Variables|/Game/Athena/Items/Consumables/C4/B_BGA_Athena_C4.B_BGA_Athena_C4_C|GetProjectileMovement1 _asb_bga_athena_c4)))
                    (:CastFailed
                      (bind _asphysics_ball_master (Utilities|Casting|CastToPhysicsBall_Master Actor)
                        (:then
                          (CallFunction|ToyEnter _asphysics_ball_master))
                        (:CastFailed
                          (bind _asfort_physics_pawn_object (Utilities|Casting|CastToFortPhysicsPawnObject Actor))
                          (PhysicsObject|ApplyImpulse _asfort_physics_pawn_object self (* (Transformation|GetActorUpVector) (* (ScalableFloat|GetValueatLevel (Variables|Default|GetLaunchVelocity)) (Variables|UserKnobs%28Creative%29|GetDynamicForceScale)))))))))))))))
    (:then_1
      (if (Networking|IsServer)
        (CallFunction|CEPlayWind Actor)))))

(event Custom|ResetDoOnceLaunch
  ;[1 1 1] Begin Overlap
  (Utilities|FlowControl|DoOnce false
    (:Completed
      (Audio|BroadcastAbilityImpactatLocation 0 (stale Custom|Actor%20Entered%20Trigger _returnvalue) (+ (Vehicle|GetVehicleVelocity (stale Custom|Actor%20Entered%20Trigger _asfort_vehicle_interface)) (* (Transformation|GetActorUpVector) (* (ScalableFloat|GetValueatLevel (Variables|Default|GetLaunchVelocity)) (Variables|UserKnobs%28Creative%29|GetDynamicForceScale)))) 777.0))))

(event Collision|EventActorBeginOverlap
  (params (OtherActor :type Actor%20Object%20Reference)))

(event EventTick
  (params (DeltaSeconds :type Float%20%28single-precision%29)))

(event Custom|Actor%20Exited%20Trigger
  (params (Actor :type Actor%20Object%20Reference))
  ;[1 1 1] End Overlap
  (bind _asfort_player_pawn (Utilities|Casting|CastToFortPlayerPawn Actor)
    (:then
      (Ability|SendGameplayEventtoActor _asfort_player_pawn "(TagName=\"Abilities.Generic.Vent.Exit\")")
      (Variables|/Script/FortniteGame.FortPawn|SetbIgnoreNextFallingDamage true _asfort_player_pawn)
      (if (Variables|Default|GetShouldBoost)
        (Projectile|RemoveGameplayCue%28Looping%29Local _asfort_player_pawn (Variables|Default|GetVentCue) (Ability|GameplayCue|MakeGameplayCueParameters :NormalizedMagnitude 0.0 :RawMagnitude 0.0 :EffectContext "()" :AggregatedSourceTags "(GameplayTags=)" :AggregatedTargetTags "(GameplayTags=)" :Location "0,0,0" :Normal "0,0,0" :GameplayEffectLevel 1 :AbilityLevel 1 :bReplicateLocationWhenUsingMinimalRepProxy false))
        (else
          (Projectile|RemoveGameplayCue%28Looping%29Local _asfort_player_pawn (Variables|Default|GetVolcanoCue) (Ability|GameplayCue|MakeGameplayCueParameters :NormalizedMagnitude 0.0 :RawMagnitude 0.0 :EffectContext "()" :AggregatedSourceTags "(GameplayTags=)" :AggregatedTargetTags "(GameplayTags=)" :Location "0,0,0" :Normal "0,0,0" :GameplayEffectLevel 1 :AbilityLevel 1 :bReplicateLocationWhenUsingMinimalRepProxy false)))))
    (:CastFailed
      (bind _asfort_vehicle_interface (Utilities|Casting|CastToFortVehicleInterface Actor)
        (:then
          (Vehicle|SetIgnoreAllFallingDamage _asfort_vehicle_interface false)
          (Vehicle|SetIgnoreNextFallingDamage _asfort_vehicle_interface true))
        (:CastFailed
          (bind _asfort_projectile_base (Utilities|Casting|CastToFortProjectileBase Actor)
            (:CastFailed
              (bind _asb_bga_athena_c4 (Utilities|Casting|CastToB_BGA_Athena_C4 Actor)
                (:then
                  (Variables|/Script/Engine.ProjectileMovementComponent|SetProjectileGravityScale (Math|Float|float*float (Variables|/Script/Engine.ProjectileMovementComponent|GetProjectileGravityScale (Variables|/Game/Athena/Items/Consumables/C4/B_BGA_Athena_C4.B_BGA_Athena_C4_C|GetProjectileMovement1 _asb_bga_athena_c4)) -1.0 (Variables|UserKnobs%28Creative%29|GetDynamicForceScale)) (Variables|/Game/Athena/Items/Consumables/C4/B_BGA_Athena_C4.B_BGA_Athena_C4_C|GetProjectileMovement1 _asb_bga_athena_c4)))
                (:CastFailed
                  (bind _asphysics_ball_master (Utilities|Casting|CastToPhysicsBall_Master Actor)))))))))))

(event OnComponentEndOverlap(TriggerVolume)
  (params (OverlappedComponent :type Primitive%20Component%20Object%20Reference) (OtherActor :type Actor%20Object%20Reference) (OtherComp :type Primitive%20Component%20Object%20Reference) (OtherBodyIndex :type Integer))
  ;[1 1 1] End Overlap
  (CallFunction|ActorExitedTrigger OtherActor))

(event Custom|CE_PlayWind
  (params (Actor :type Actor%20Object%20Reference))
  (bind _sm_hvac_fx (Variables|Default|GetSM_HVAC_FX))
  (bind _returnvalue (Transformation|GetWorldRotation _sm_hvac_fx))
  (bind _fx_mid (Variables|Default|GetFX_MID))
  ;[0 0 0] FX
  (Utilities|FlowControl|Sequence
    (:then_0
      (Variables|Default|GetFX_MID
        (:then)
        (:else
          (bind _returnvalue_2 (Rendering|Material|CreateDynamicMaterialInstance (Variables|Default|GetSM_HVAC_FX) 0 "/Game/Athena/Environments/Blueprints/DudeBro/HVAC_FX/FX_Wind.FX_Wind" "None" :!class /Script/Engine.PrimitiveComponent))
          (Variables|Default|SetFX_MID _returnvalue_2))))
    (:then_1
      (bind (_direction _bp_speeduptimer _bp_wpo_movingup _bp_intensity) (Timeline|Timeline_0 PlayFromStart)
        (:Update
          (Rendering|Material|SetScalarParameterValue _fx_mid "BP_SpeedUpTimer" _bp_speeduptimer :!class /Script/Engine.MaterialInstanceDynamic)
          (Rendering|Material|SetScalarParameterValue _fx_mid "BP_WPO_MovingUP" (Math|Float|Min(Float) (* _bp_wpo_movingup (Variables|UserKnobs%28Creative%29|GetFX%20Height%20Multiplier)) (Variables|UserKnobs%28Creative%29|GetMax%20WPO_MovingUpHeight)) :!class /Script/Engine.MaterialInstanceDynamic)
          (Rendering|Material|SetScalarParameterValue _fx_mid "BP_Intensity" (* _bp_intensity (Variables|UserKnobs%28Creative%29|GetFX%20Intensity%20Multiplier)) :!class /Script/Engine.MaterialInstanceDynamic))))
    (:then_2
      (bind _asfort_player_pawn (Utilities|Casting|CastToFortPlayerPawn Actor))
      (if (Variables|Default|GetbEnableVFX)
        (GameplayCue|ExecuteGameplayCue%28Burst%29Local _asfort_player_pawn (Variables|Default|GetWindCue) (Ability|GameplayCue|MakeGameplayCueParameters :Normal (Math|Rotator|GetRotationXVector _returnvalue)) true)
        (Effects|SpawnEmitteratLocation "/Game/Athena/Environments/Blueprints/DudeBro/HVAC_FX/P_WindBasement.P_WindBasement" (Transformation|GetWorldLocation (Variables|Transformation|GetRootComponent _asfort_player_pawn)) _returnvalue))
      (if (Variables|Default|GetbEnableSFX)
        (GameplayCue|ExecuteGameplayCueOnActor(Burst) self (Variables|Default|GetGCJumpedOn) (Ability|GameplayCue|MakeGameplayCueParameters 0.0))))))

(fn GetDynamicDestructionStaticMeshComponents
  (return (Utilities|Array|MakeArray (Variables|Debug|GetStaticMeshComponent)) (Utilities|Array|MakeArray (Variables|Default|GetSM_HVAC_FX)) (Utilities|Array|MakeArray (Variables|Debug|GetStaticMeshComponent))))

(fn OnRep_isDisabled
  (if (Variables|Default|GetisDisabled)
    (Components|Activation|Deactivate (Variables|Default|GetLoopingAmbAudio))
    (Rendering|SetActorHiddenInGame self true)
    (Collision|SetActorEnableCollision false)))

(fn ProjEnter
  (params (Projectile :type Fort%20Projectile%20Base%20Object%20Reference))
  (Utilities|IsValid Projectile
    (:Is%20Valid
      (if (not (GameplayTags|HasMatchingGameplayTag Projectile "(TagName=\"Projectile.IgnoreHVACVelocityManipulation\")"))
        (Variables|/Script/Engine.MovementComponent|SetVelocity (Math|Vector|MakeVector (+ (.x (Math|Vector|Lerp(Vector) (* (Transformation|GetActorUpVector) (ScalableFloat|GetValueatLevel (Variables|Default|GetLaunchVelocity))) (Variables|/Script/Engine.MovementComponent|GetVelocity (Variables|/Script/FortniteGame.FortProjectileBase|GetProjectileMovementComponent Projectile)) 0.58)) (Math|Random|RandomFloatinRange -1.0 1.0)) (+ (.y (Math|Vector|Lerp(Vector) (* (Transformation|GetActorUpVector) (ScalableFloat|GetValueatLevel (Variables|Default|GetLaunchVelocity))) (Variables|/Script/Engine.MovementComponent|GetVelocity (Variables|/Script/FortniteGame.FortProjectileBase|GetProjectileMovementComponent Projectile)) 0.58)) (Math|Random|RandomFloatinRange -1.0 1.0)) (Math|Float|Clamp(Float) (.z (Math|Vector|Lerp(Vector) (* (Transformation|GetActorUpVector) (ScalableFloat|GetValueatLevel (Variables|Default|GetLaunchVelocity))) (Variables|/Script/Engine.MovementComponent|GetVelocity (Variables|/Script/FortniteGame.FortProjectileBase|GetProjectileMovementComponent Projectile)) 0.58)) (ScalableFloat|GetValueatLevel (Variables|Default|GetMinProjLaunchVelocity)) (* (ScalableFloat|GetValueatLevel (Variables|Default|GetLaunchVelocity)) (Variables|UserKnobs%28Creative%29|GetDynamicForceScale)))) (Variables|/Script/FortniteGame.FortProjectileBase|GetProjectileMovementComponent Projectile))))))

(fn ProjExit
  (params (Projectile :type Fort%20Projectile%20Base%20Object%20Reference)))

(fn ToyExit
  (params (Projectile :type Physics%20Ball%20Master%20Object%20Reference)))

(fn ToyEnter
  (params (Projectile :type Physics%20Ball%20Master%20Object%20Reference))
  (Variables|/Script/Engine.MovementComponent|SetVelocity (Math|Vector|Lerp(Vector) (* (Transformation|GetActorUpVector) (* (ScalableFloat|GetValueatLevel (Variables|Default|GetLaunchVelocity)) (Variables|UserKnobs%28Creative%29|GetDynamicForceScale))) (Variables|/Script/Engine.MovementComponent|GetVelocity (Variables|/Game/Building/ActorBlueprints/Prop/PhysicsBall_Master.PhysicsBall_Master_C|GetProjectileMovement Projectile)) 0.5) (Variables|/Game/Building/ActorBlueprints/Prop/PhysicsBall_Master.PhysicsBall_Master_C|GetProjectileMovement Projectile)))

(fn ConstructionScript)
