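; DudeBro vortex mesh actor that pairs a static mesh with ambient audio reflecting event progression.
; UpdateCachedTime and UpdateDrainAmount set audio parameters for completion ratio or drain level, switch the Progression stage and play the sound.
; UpdateDrainAmount also sets the mesh's OverallOpacity to half.
; On Cube Final Crack waits briefly, calls UpdateDrainAmount with zero, then runs a timeline interpolating opacity between half and full; StopSounds halts the audio.
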(event EventBeginPlay)

(event Custom|StopSounds
  (Audio|Components|Audio|Stop (Variables|Default|GetAmbAudio)))

(event Collision|EventActorBeginOverlap
  (params (OtherActor :type Actor%20Object%20Reference)))

(event EventTick
  (params (DeltaSeconds :type Float%20%28single-precision%29)))

(event Custom|On%20Cube%20Final%20Crack
  (Utilities|FlowControl|Delay 0.5)
  (CallFunction|UpdateDrainAmount 0.0)
  (bind (_direction _lerp) (Timeline|Fade%20down%20Vertex PlayFromStart)
    (:Update
      (Rendering|Material|SetScalarParameterValueonMaterials (Variables|Default|GetStaticMesh) "OverallOpacity" (Math|Float|Lerp 0.5 1.0 _lerp)))))

(fn UpdateCachedTime
  (params (Time :type Float%20%28double-precision%29) (RatioComplete :type Float%20%28double-precision%29))
  (Audio|Parameter|SetIntegerParameter (Variables|Default|GetAmbAudio) "Progression" 3 :!class /Script/AudioExtensions.AudioParameterControllerInterface)
  (Audio|Parameter|SetFloatParameter (Variables|Default|GetAmbAudio) "RatioComplete" RatioComplete :!class /Script/AudioExtensions.AudioParameterControllerInterface)
  (Audio|Components|Audio|Play (Variables|Default|GetAmbAudio) 0.0))

(fn UpdateDrainAmount
  (params (Amount :type Float%20%28double-precision%29))
  (Audio|Parameter|SetFloatParameter (Variables|Default|GetAmbAudio) "DrainAmount" (- 1.0 Amount) :!class /Script/AudioExtensions.AudioParameterControllerInterface)
  (Audio|Parameter|SetFloatParameter (Variables|Default|GetAmbAudio) "DrainAmountInv" Amount :!class /Script/AudioExtensions.AudioParameterControllerInterface)
  (Audio|Parameter|SetIntegerParameter (Variables|Default|GetAmbAudio) "Progression" 4 :!class /Script/AudioExtensions.AudioParameterControllerInterface)
  (Audio|Components|Audio|Play (Variables|Default|GetAmbAudio) 0.0)
  (Rendering|Material|SetScalarParameterValueonMaterials (Variables|Default|GetStaticMesh) "OverallOpacity" 0.5))

(fn ConstructionScript)
